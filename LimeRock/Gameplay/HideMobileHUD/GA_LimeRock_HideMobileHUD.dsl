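; Gameplay ability for the Lime Rock cinematic event that hides the mobile touch-controls joystick while it is active.
; On activation it calls the Lime Rock function library to hide the HUD element tagged HUD.Bacchus.Joystick for the local player.
; When the ability ends, whether cancelled or not, it shows the joystick again.
; Both calls run only where cosmetic events can execute.
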(event Ability|EventActivateAbility
  (Networking|CanExecuteCosmeticEvents
    (:True
      (Class|BPLimeRockFunctionLibrary|HideHUDForLocalPlayer "(GameplayTags=((TagName=\"HUD.Bacchus.Joystick\")))" true "Hiding touch controls joystick during lime rock cinematic event."))))

(event Ability|EventOnEndAbility
  (params (bWasCancelled :type Boolean))
  (Networking|CanExecuteCosmeticEvents
    (:True
      (Class|BPLimeRockFunctionLibrary|HideHUDForLocalPlayer "(GameplayTags=((TagName=\"HUD.Bacchus.Joystick\")))" false "Showing touch controls joystick after lime rock cinematic event."))))
